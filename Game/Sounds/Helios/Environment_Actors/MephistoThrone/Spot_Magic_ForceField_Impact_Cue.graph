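# Sound cue for a magic force field impact, filed under the MephistoThrone environment actor sounds.
# A short randomized delay feeds a mixer of two layers: one of six random impact waves through a modulator, and a slightly louder, further-delayed pick of two others with narrow pitch variation.
# It uses bullet-impact attenuation and the impacts sound class, with a concurrency override that caps simultaneous instances.
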SoundCue /Game/Sounds/Helios/Environment_Actors/MephistoThrone/Spot_Magic_ForceField_Impact_Cue
root=Delay_0
nodes [
 {id=Delay_0, children [Mixer_0], properties {delayMax=0.02, delayMin=0.01}}
 {id=Delay_1, children [Modulator_2], properties {delayMax=0.02, delayMin=0.01}}
 {id=Mixer_0, children [Modulator_0,Delay_1], properties {inputVolume [1,1.2]}}
 {id=Modulator_0, children [Random_0]}
 {id=Modulator_2, children [Random_3], properties {pitchMax=1.02, pitchMin=0.97, volumeMin=1}}
 {
  id=Random_0
  children [WavePlayer_1, WavePlayer_2, WavePlayer_3, WavePlayer_4, WavePlayer_9, WavePlayer_10]
  properties {weights [1,1,1,1,1,1]}
 }
 {id=Random_3, children [WavePlayer_13, WavePlayer_14], properties {weights [1,1]}}
 {id=WavePlayer_1, properties {soundWaveAssetPtr=&/ElbowChat_Forcefield_Impact_01}}
 {id=WavePlayer_10, properties {soundWaveAssetPtr=&/ElbowChat_Forcefield_Impact_06}}
 {id=WavePlayer_13, properties {soundWaveAssetPtr=&/ElbowChat_Forcefield_Impact_08}}
 {id=WavePlayer_14, properties {soundWaveAssetPtr=&/ElbowChat_Forcefield_Impact_07}}
 {id=WavePlayer_2, properties {soundWaveAssetPtr=&/ElbowChat_Forcefield_Impact_02}}
 {id=WavePlayer_3, properties {soundWaveAssetPtr=&/ElbowChat_Forcefield_Impact_03}}
 {id=WavePlayer_4, properties {soundWaveAssetPtr=&/ElbowChat_Forcefield_Impact_04}}
 {id=WavePlayer_9, properties {soundWaveAssetPtr=&/ElbowChat_Forcefield_Impact_05}}
]
properties {
 attenuationSettings=/Game/Sounds/Attenuation/Fort_Bullet_Impact_Att
 bOverrideConcurrency=true
 maxDistance=5300
 soundClassObject=&Fort_Impacts
 volumeMultiplier=1
 concurrencyOverrides {maxCount=2, voiceStealReleaseTime=0.5}
}
