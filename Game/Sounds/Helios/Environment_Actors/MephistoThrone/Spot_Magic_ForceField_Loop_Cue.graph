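# Looping spot ambience cue for a magic force field, filed with the Mephisto Throne environment actor sounds.
# A single looping energy field wave feeds a Mixer root, whose second input is empty, at reduced volume under the Fort_Gadgets sound class.
# Attenuation override uses logarithmic falloff with distance filtering and listener focus.
# Concurrency override caps instances and scales their volume, and submix sends are disabled.
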SoundCue /Game/Sounds/Helios/Environment_Actors/MephistoThrone/Spot_Magic_ForceField_Loop_Cue
root=Mixer_0
nodes [
 {id=Mixer_0, children [WavePlayer_1,null], properties {inputVolume [1,1.2]}}
 {
  id=WavePlayer_1
  properties {bLooping=true, soundWaveAssetPtr=&/ElbowChat_Energy_Field_Loop}
 }
]
properties {
 bEnableSubmixSends=false
 bOverrideAttenuation=true
 bOverrideConcurrency=true
 maxDistance=1650
 soundClassObject=&Fort_Gadgets
 volumeMultiplier=0.25
 attenuationOverrides {
  bAttenuateWithLPF=true
  bEnableFocusInterpolation=true
  bEnableListenerFocus=true
  distanceAlgorithm=Logarithmic
  falloffDistance=1500
  hPFFrequencyAtMax=350
  lPFRadiusMax=1500
  lPFRadiusMin=1000
  nonFocusVolumeAttenuation=0.75
  stereoSpread=150
  attenuationShapeExtents {x=150}
 }
 concurrencyOverrides {maxCount=3, voiceStealReleaseTime=0.5, volumeScale=0.75, volumeScaleAttackTime=0.25}
}
