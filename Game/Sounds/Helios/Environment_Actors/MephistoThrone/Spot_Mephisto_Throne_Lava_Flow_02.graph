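# Looping ambient spot sound cue for a lava flow at the Mephisto throne environment actor.
# A mixer layers two looping lava waves, a volcano lava fall and a large Mephisto lava flow.
# Two nested, endlessly looping envelopers of different lengths add slow pitch modulation and a gentle volume swell.
# Overrides set capsule attenuation with listener focus and allow one instance, rejecting new ones; the cue also sends to reverb.
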SoundCue /Game/Sounds/Helios/Environment_Actors/MephistoThrone/Spot_Mephisto_Throne_Lava_Flow_02
root=Enveloper_2
nodes [
 {
  id=Enveloper_0
  children [Mixer_1]
  properties {
   bLoop=true
   bLoopIndefinitely=true
   loopEnd=30
   pitchCurve {editorCurveData {keys [{interpMode=RCIM_Cubic, time=0}, <at 15.9468 0.9181 RCIM_Cubic>, {interpMode=RCIM_Cubic, time=30}]}}
   volumeCurve {editorCurveData {keys [<at 0 1>]}}
  }
 }
 {
  id=Enveloper_2
  children [Enveloper_0]
  properties {
   bLoop=true
   bLoopIndefinitely=true
   loopEnd=20
   pitchCurve {editorCurveData {keys [{time=0}]}}
   volumeCurve {editorCurveData {keys [
      {arriveTangent=-0.01, interpMode=RCIM_Cubic, leaveTangent=-0.01, time=0, value=1}
      <at 10 0.9 RCIM_Cubic>
      {arriveTangent=0.01, interpMode=RCIM_Cubic, leaveTangent=0.01, time=20, value=1}
     ]}}
  }
 }
 {id=Mixer_1, children [WavePlayer_8, WavePlayer_14], properties {inputVolume [1,1]}}
 {
  id=WavePlayer_14
  properties {bLooping=true, soundWaveAssetPtr=/Game/Sounds/Helios/Environment_Actors/Mephisto_Hades/Spot_Mephisto_LavaFlow_Large_Loop_02}
 }
 {id=WavePlayer_8, properties {bLooping=true, soundWaveAssetPtr=/Game/Athena/Sounds/Ambient/Volcano/Amb_Volcano_Lava_Fall_02}}
]
properties {
 bOverrideAttenuation=true
 bOverrideConcurrency=true
 maxDistance=1650
 soundClassObject=&Fort_Ambient_Indoors_Athena
 volumeMultiplier=0.4
 attenuationOverrides {
  attenuationShape=Capsule
  bEnableFocusInterpolation=true
  bEnableListenerFocus=true
  falloffDistance=1500
  nonFocusVolumeAttenuation=0.4
  stereoSpread=0
  attenuationShapeExtents=<150 50>
 }
 concurrencyOverrides {maxCount=1, resolutionRule=PreventNew}
 soundSubmixSends [{minSendLevel=0.6, soundSubmix=&Fortnite_Reverb_Default}]
}
