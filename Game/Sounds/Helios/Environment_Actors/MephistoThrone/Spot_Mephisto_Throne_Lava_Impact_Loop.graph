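# Looping ambient spot cue for lava impact at the Mephisto throne, built from reused fire and sizzle loops.
# A mixer layers a pitch-randomized fire bed (random urn loop plus brazier) with a random forge sizzle and a barbecue sizzle.
# Nested looping envelopers slowly vary each sizzle's pitch and volume on differing periods.
# The root randomly picks one of three fixed-pitch versions, with short-range focused attenuation, limited concurrency and reverb send.
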SoundCue /Game/Sounds/Helios/Environment_Actors/MephistoThrone/Spot_Mephisto_Throne_Lava_Impact_Loop
root=Random_4
nodes [
 {
  id=Enveloper_0
  children [Random_3]
  properties {
   bLoop=true
   bLoopIndefinitely=true
   loopEnd=30
   pitchCurve {editorCurveData {keys [{interpMode=RCIM_Cubic, time=0}, <at 15.9468 0.9181 RCIM_Cubic>, {interpMode=RCIM_Cubic, time=30}]}}
   volumeCurve {editorCurveData {keys [<at 0 1>]}}
  }
 }
 {
  id=Enveloper_2
  children [Enveloper_0]
  properties {
   bLoop=true
   bLoopIndefinitely=true
   loopEnd=20
   pitchCurve {editorCurveData {keys [{time=0}]}}
   volumeCurve {editorCurveData {keys [
      {arriveTangent=-0.01, interpMode=RCIM_Cubic, leaveTangent=-0.01, time=0, value=1}
      <at 10 0.9 RCIM_Cubic>
      {arriveTangent=0.01, interpMode=RCIM_Cubic, leaveTangent=0.01, time=20, value=1}
     ]}}
  }
 }
 {
  id=Enveloper_4
  children [Enveloper_5]
  properties {
   bLoop=true
   bLoopIndefinitely=true
   loopEnd=20
   pitchCurve {editorCurveData {keys [{time=0}]}}
   volumeCurve {editorCurveData {keys [
      {arriveTangent=-0.01, interpMode=RCIM_Cubic, time=0, value=1}
      <at 10 0.8 RCIM_Cubic>
      {interpMode=RCIM_Cubic, leaveTangent=0.01, time=20, value=1}
     ]}}
  }
 }
 {
  id=Enveloper_5
  children [WavePlayer_6]
  properties {
   bLoop=true
   bLoopIndefinitely=true
   loopEnd=30
   pitchCurve {editorCurveData {keys [{interpMode=RCIM_Cubic, time=0}, <at 15.9468 1.2 RCIM_Cubic>, {interpMode=RCIM_Cubic, time=30}]}}
   volumeCurve {editorCurveData {keys [<at 0 1>]}}
  }
 }
 {id=Mixer_0, children [Modulator_0,Enveloper_2,Enveloper_4], properties {inputVolume [0.8,0.25,0.8]}}
 {id=Mixer_1, children [Random_0,WavePlayer_1], properties {inputVolume [1,2.5]}}
 {id=Modulator_0, children [Mixer_1], properties {pitchMax=1.1, pitchMin=0.9}}
 {id=Modulator_3, children [Mixer_0], properties {pitchMax=1.1, pitchMin=1.1}}
 {id=Modulator_4, children [Mixer_0], properties {pitchMax=1.2, pitchMin=1.2}}
 {id=Modulator_5, children [Mixer_0], properties {pitchMax=0.9, pitchMin=0.9}}
 {id=Random_0, children [WavePlayer_0,WavePlayer_4,WavePlayer_5], properties {weights [1,1,1]}}
 {id=Random_3, children [WavePlayer_20, WavePlayer_21], properties {weights [1,1]}}
 {id=Random_4, children [Modulator_4,Modulator_5,Modulator_3], properties {weights [1,1,1]}}
 {id=WavePlayer_0, properties {bLooping=true, soundWaveAssetPtr=/Game/Sounds/Trans_AmbientLoops/Fire/Fire_Urn_01}}
 {
  id=WavePlayer_1
  properties {bLooping=true, soundWaveAssetPtr=/Game/Sounds/Artemis/Environment_Actors/ColiseumBrazierFire/Spot_Coliseum_Brazier_Fire_Close_Mono}
 }
 {
  id=WavePlayer_20
  properties {bLooping=true, soundWaveAssetPtr=/Game/Sounds/Helios/Environment_Actors/Castle_Forge/Spot_Forge_MoltenMetal_Sizzle_01}
 }
 {
  id=WavePlayer_21
  properties {bLooping=true, soundWaveAssetPtr=/Game/Sounds/Helios/Environment_Actors/Castle_Forge/Spot_Forge_MoltenMetal_Sizzle_02}
 }
 {id=WavePlayer_4, properties {bLooping=true, soundWaveAssetPtr=/Game/Sounds/Trans_AmbientLoops/Fire/Fire_Urn_02}}
 {id=WavePlayer_5, properties {bLooping=true, soundWaveAssetPtr=/Game/Sounds/Trans_AmbientLoops/Fire/Fire_Urn_03}}
 {id=WavePlayer_6, properties {bLooping=true, soundWaveAssetPtr=/Game/Sounds/Helios/Environment_Actors/BBQSizzle/Spot_Barbecue_Sizzle_Loop}}
]
properties {
 bOverrideAttenuation=true
 bOverrideConcurrency=true
 maxDistance=750
 soundClassObject=&Fort_Ambient_Indoors_Athena
 volumeMultiplier=0.2
 attenuationOverrides {
  bEnableFocusInterpolation=true
  bEnableListenerFocus=true
  distanceAlgorithm=Logarithmic
  falloffDistance=650
  focusAttackInterpSpeed=2
  focusAzimuth=15
  nonFocusAzimuth=30
  nonFocusVolumeAttenuation=0.25
  stereoSpread=10
  attenuationShapeExtents {x=100}
 }
 concurrencyOverrides {maxCount=3}
 soundSubmixSends [{minSendLevel=0.6, soundSubmix=&Fortnite_Reverb_Default}]
}
